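# Translucent UI material for the Puppeteer boss bar, a progress bar with gradient fill and outline.
# A 9-sliced box SDF minus an offset circle forms the notched shape, and SDF exclusion of both yields the outline.
# A step of horizontal UV against the Progress scalar clips the fill.
# Two color-pair gradients tint fill and outline and are added, while the max of both masks forms the combined opacity.
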Material /PuppeteerBoss/UI/Materials/M_UI_Puppeteer_Bar
graph {
 properties {bAutomaticallySetUsageInEditor=false, blendMode=BLEND_Translucent, materialDomain=MD_UI, usedShadingModels=Default Lit}
 nodes [
  n1 = VectorParameter() {parameterName=Fill Color 1, defaultValue=<rgba 1 0.2755 0 1>}
  n2 = VectorParameter() {parameterName=Fill Color 2, defaultValue=<rgba 0 0.8768 1 1>}
  n3 = MaterialFunctionCall(UVs=TextureCoordinate(), GradientStart=Constant() {r=0.1}, GradientEnd=Constant() {r=0.9}) {materialFunction=&MF_UI_LinearGradient}
  n4 = VectorParameter() {parameterName=Outline Color 1, defaultValue=<rgba 0.9819 0 1 1>}
  n5 = VectorParameter() {parameterName=Outline Color 2, defaultValue=<rgba 0.158 1 0.0076 1>}
  Add(
   A=Multiply(A=LinearInterpolate(A=@n1.RGB, B=@n2.RGB, Alpha=@n3.Result), B=NamedRerouteUsage())
   B=Multiply(A=LinearInterpolate(A=@n4.RGB, B=@n5.RGB, Alpha=@n3.Result), B=NamedRerouteUsage())
  )
  n6 = MaterialFunctionCall() {materialFunction=&GetUserInterfaceUV}
  n7 = Constant2Vector() {g=1, r=1}
  n8 = MaterialFunctionCall(UVs=MaterialFunctionCall!(UVs=@n6.Normalized UV) {materialFunction=/FortUILibrary/MaterialFunctions/Basics/MF_UI_9SliceUVs}.UVs, Size=@n7) {materialFunction=&MF_UI_Box}
  n9 = ComponentMask(None=@n6.Normalized UV) {g=true}
  n10 = MaterialFunctionCall(UVs=AppendVector(A=Add(A=@n9) {constB=0.875}, B=ComponentMask(None=@n6.Normalized UV) {r=true}), Size=Multiply(A=@n7.R) {constB=1.3}) {materialFunction=&MF_UI_Circle}
  n11 = NamedRerouteDeclaration(None=Subtract(A=@n8.Fill, B=@n10.Fill)) {name=Fill, nodeColor=<rgba 0 0 0 1>}
  n12 = NamedRerouteDeclaration(None=Step(Y=ComponentMask(None=@n6.Normalized UV) {r=true}, X=ScalarParameter() {defaultValue=0.5, parameterName=Progress})) {name=Progress, nodeColor=<rgba 0 0 0 1>}
  n13 = MaterialFunctionCall!(
   DistanceField=MaterialFunctionCall!(DistanceFieldA=@n10.SDF, DistanceFieldB=@n8.SDF, BlendAmount) {materialFunction=/FortUILibrary/MaterialFunctions/Shapes/Modifiers/MF_UI_SDF_Exclusion}.Exact
   Threshold
   Thickness
   Alignment=Constant()
   Smoothing
  ) {materialFunction=/FortUILibrary/MaterialFunctions/Shapes/Modifiers/MF_UI_ShapeOutline}
  Max(
   A=NamedRerouteDeclaration(None=Multiply(A=@n11, B=@n12)) {name=ProgressMask, nodeColor=<rgba 0 0 0 1>}
   B=NamedRerouteDeclaration(None=Subtract(A=@n13.Outline, B=@n10.Fill)) {name=OpacityMaskOutline, nodeColor=<rgba 0 0 0 1>}
  )
 ]
}
