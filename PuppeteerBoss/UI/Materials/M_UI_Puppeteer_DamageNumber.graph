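# UI material for the Puppeteer boss that, going by its name, backs damage numbers with a two-tone circular graphic.
# An inner circle is filled with a rotated linear gradient from black to red.
# A pink-magenta outer ring comes from subtracting the inner circle's fill mask from a larger one.
# The two masked layers are added together, with shapes and gradient built procedurally by shared UI material functions.
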Material /PuppeteerBoss/UI/Materials/M_UI_Puppeteer_DamageNumber
graph {
 properties {bAutomaticallySetUsageInEditor=false, materialDomain=MD_UI, usedShadingModels=Default Lit}
 nodes [
  n1 = Constant3Vector() {constant=<rgba 0 0 0 1>}
  n2 = Constant3Vector() {constant=<rgba 0.9294 0.1098 0.1412 1>}
  n3 = MaterialFunctionCall(UVs=TextureCoordinate(), GradientStart=Constant(), RotationAngle=Constant() {r=0.125}) {materialFunction=&MF_UI_LinearGradient}
  n4 = MaterialFunctionCall(UVs=TextureCoordinate(), Size=Constant() {r=0.6}, GlowMax=Constant()) {materialFunction=&MF_UI_Circle}
  n5 = MaterialFunctionCall(UVs=TextureCoordinate(), Size=Constant() {r=1}, GlowMax=Constant()) {materialFunction=&MF_UI_Circle}
  Add(
   A=LinearInterpolate(B=LinearInterpolate(A=@n1, B=@n2, Alpha=@n3.Result), Alpha=@n4.Fill)
   B=LinearInterpolate(B=Constant3Vector() {constant=<rgba 1 0 0.6485 1>}, Alpha=Subtract(A=@n5.Fill, B=@n4.Fill))
  )
 ]
}
